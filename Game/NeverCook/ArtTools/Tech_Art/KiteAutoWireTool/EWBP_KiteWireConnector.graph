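# Editor utility widget for tech art that builds and edits BP_AutoWireActor wires between kite or inflatable actors and ground anchors.
# Grab Selected Actor sorts the selection by _Kite/_Inflatable name suffix and looks up any existing wire using them.
# Socket dropdowns list mesh sockets, while a KiteData asset supplies material and thickness.
# Update Wire reinitializes the wire, first spawning one if needed at the anchor, or at a traced or computed ground point.
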EditorUtilityWidgetBlueprint /Game/NeverCook/ArtTools/Tech_Art/KiteAutoWireTool/EWBP_KiteWireConnector
graphs [
 ConvertSelectedActor {nodes [
   n1 = ConvertSelectedActor!() -> (then, Target:Actor Object Reference) {is=title K2Node_FunctionEntry}
   > Is Valid(InputObject=@n1)
   n2 = GetComponentByClass!(self:Actor Object Reference=@n1, ComponentClass:Actor Component Class Reference=&StaticMeshComponent) -> (ReturnValue:Static Mesh Component Object Reference)
   > n3 = Is Valid(InputObject=@n2)
   n4 = Cast To StaticMeshActor!(execute=@n3.Is Not Valid, Object=@n1) -> (then, CastFailed, AsStatic Mesh Actor:Static Mesh Actor Object Reference) {is=title K2Node_DynamicCast}
   n5 = Get StaticMeshComponent(self=@n4)
   > Return Node!(execute, ReturnValue:Static Mesh Component Object Reference=@n5, IsValid:bool=IsValid(Object=@n5))
   Return Node(execute=@n4.CastFailed, IsValid=false)
   Return Node(execute=@n3.Is Valid, ReturnValue=@n2, IsValid=true)
  ]}
 EventGraph {nodes [
   "On Property Changed (RotationProperty)"!() -> (OutputDelegate, then, PropertyName:Name=None) {is=title K2Node_ComponentBoundEvent}
   > n1 = Get!(execute:Exec) -> (then:Exec, else:Exec, CurrentSpawnedWire:BP Auto Wire Actor Object Reference)
   n2 = Get(execute=[@n40.then, @n38.then_2, @n1.then])
   n3 = Get CurrentSpawnedWire() -> (CurrentSpawnedWire:BP Auto Wire Actor Object Reference)
   Initialize!(
    execute=[@n6.then, @n13.then, @n2.then]
    self:BP Auto Wire Actor Object Reference=@n3
    KiteData:BP Kite Wire Data Object Reference=Get KiteData() -> (KiteData:BP Kite Wire Data Object Reference)
    GroundWireConnectionMesh:Static Mesh Object Reference=Get AnchorMesh() -> (AnchorMesh:Static Mesh Object Reference)
    WireMaterial:Material Instance Object Reference=Get WireMaterial() -> (WireMaterial:Material Instance Object Reference)
    Thickness:double=Get Thickness() -> (Thickness:double)
    KiteBaloonMeshActor:Actor Object Reference=Get KiteBallonActor() -> (KiteBallonActor:Actor Object Reference)
    KiteSocketAnchorName:Name=Get KiteSocketAnchorName() -> (KiteSocketAnchorName:Name=None)
    GroundSocketAnchorName:Name=Get GroundSocketAnchorName() -> (GroundSocketAnchorName:Name=None)
    GroundStaticMeshActor:Actor Object Reference=Get AnchorActor() -> (AnchorActor:Actor Object Reference)
    ForceKiteBalloonLookAt:bool=Get AutoLookAt() -> (AutoLookAt:bool)
    bInflatable:bool=Get bInflatable() -> (bInflatable:bool)
   ) -> (then)
   > UpdateWire(self:BP Auto Wire Actor Object Reference=@n3)
   n4 = Get!(execute=@n2.else) -> (then, else, AnchorActor:Actor Object Reference)
   > n5 = SpawnActor BP Auto Wire Actor!(
    execute:Exec
    WorldContextObject:Object Reference=Self-Reference()
    Class:Actor Class Reference=/Game/ArtTools/Tech_Art/KiteAutoWireTool/BP_AutoWireActor.BP_AutoWireActor_C
    SpawnTransform_Location:Vector=Get Actor Location(self=@n4)
    SpawnTransform_Rotation:Rotator="0, 0, 0"
    SpawnTransform_Scale:Vector="1,1,1"
    CollisionHandlingOverride:ESpawnActorCollisionHandlingMethod Enum=Undefined
    TransformScaleMethod:ESpawnActorScaleMethod Enum=MultiplyWithRoot
    Owner:Actor Object Reference
    Instigator:Pawn Object Reference
   ) -> (then:Exec, ReturnValue:BP Auto Wire Actor Object Reference) {is=title K2Node_SpawnActorFromClass}
   > n6 = Set CurrentSpawnedWire(CurrentSpawnedWire:BP Auto Wire Actor Object Reference=@n5) -> (Output_Get:BP Auto Wire Actor Object Reference)
   n7 = Get Actor Transform!(self=Get KiteBallonActor()) -> (
    ReturnValue_Location_X:double
    ReturnValue_Location_Y:double
    ReturnValue_Location_Z:double
    ReturnValue_Rotation:Rotator
    ReturnValue_Scale:Vector
   )
   n8 = float * float(A=@n7, B=-0.5) {is=title K2Node_PromotableOperator}
   n9 = Line Trace For Objects!(
    execute=@n4.else
    WorldContextObject:Object Reference
    Start:Vector=Get Actor Transform!(self=Get KiteBallonActor()) -> (ReturnValue_Location:Vector, ReturnValue_Rotation, ReturnValue_Scale)
    End_X:double=@n8
    End_Y:double=@n7.ReturnValue_Location_Y
    End_Z:double=0
    ObjectTypes:Array of EObjectTypeQuery Enums=Make Array("[0]":EObjectTypeQuery Enum=ObjectTypeQuery1) -> (Array:Array of EObjectTypeQuery Enums)
    bTraceComplex:bool=false
    ActorsToIgnore:Array of Actor Object References
    DrawDebugType:EDrawDebugTrace Enum=ForDuration
    bIgnoreSelf:bool=true
    TraceColor:Linear Color Structure=(R=1,G=0,B=0,A=1)
    TraceHitColor:Linear Color Structure=(R=0,G=1,B=0,A=1)
    DrawTime:float=5
   ) -> (
    then
    OutHit_bBlockingHit:bool
    OutHit_bInitialOverlap:bool
    OutHit_Time:float
    OutHit_Distance:float
    OutHit_Location_X:double
    OutHit_Location_Y:double
    OutHit_Location_Z:double
    OutHit_ImpactPoint:Vector
    OutHit_Normal:Vector
    OutHit_ImpactNormal:Vector
    OutHit_PhysMat:Physical Material Object Reference
    OutHit_HitActor:Actor Object Reference
    OutHit_HitComponent:Primitive Component Object Reference
    OutHit_HitBoneName:Name=None
    OutHit_BoneName:Name=None
    OutHit_HitItem:int=0
    OutHit_ElementIndex:int=0
    OutHit_FaceIndex:int=0
    OutHit_TraceStart:Vector
    OutHit_TraceEnd:Vector
    ReturnValue:bool
   )
   n10 = GetDisplayName(Object=Get KiteBallonActor())
   n11 = AND Boolean(A=@n9.ReturnValue, B=NOT Boolean(A=Contains(SearchIn=@n10, Substring=_Inflatable, bSearchFromEnd=true)))
   > n12 = SpawnActor BP Auto Wire Actor!(
    execute
    WorldContextObject=Self-Reference()
    Class=&BP_AutoWireActor_C
    SpawnTransform_Location_X:double=SelectFloat(A=@n9.OutHit_Location_X, B=@n8, bPickA=@n11)
    SpawnTransform_Location_Y:double=SelectFloat(A=@n9.OutHit_Location_Y, B=@n7.ReturnValue_Location_Y, bPickA=@n11)
    SpawnTransform_Location_Z:double=SelectFloat(A=@n9.OutHit_Location_Z, B=0, bPickA=@n11)
    SpawnTransform_Rotation="0, 0, 0"
    SpawnTransform_Scale="1,1,1"
    CollisionHandlingOverride=Undefined
    TransformScaleMethod=MultiplyWithRoot
    Owner
    Instigator
   ) -> (then, ReturnValue)
   > n13 = Set CurrentSpawnedWire(CurrentSpawnedWire=@n12)
   n14 = "On Clicked (GrabActorButton)"() {is=title K2Node_ComponentBoundEvent}
   Set AnchorActor(execute=[@n14.then, @n42.then], AnchorActor:Actor Object Reference) -> (Output_Get:Actor Object Reference)
   > Set KiteBallonActor(KiteBallonActor:Actor Object Reference) -> (Output_Get:Actor Object Reference)
   > n15 = Sequence!(execute) -> (then_0, then_1, then_2:Exec)
   > n16 = GetSelectedActors()
   > n17 = For Each Loop(Array:Array of Actor Object References=@n16) -> (Array Element:Actor Object Reference)
   n18 = Branch(execute=@n17.Completed, Condition=IsValid(Object=Get KiteBallonActor()))
   > n19 = Get!(execute) -> (then, else, CurrentSpawnedWire:BP Auto Wire Actor Object Reference)
   > n20 = Branch(Condition="Equal (Object)"(
     A=Get KiteBalloonMeshActor!(self:BP Auto Wire Actor Object Reference=@n19) -> (KiteBalloonMeshActor:Actor Object Reference)
     B=Get KiteBallonActor()
    ))
   n21 = GetAllActorsOfClass!(
    execute=[@n20.else, @n19.else]
    WorldContextObject:Object Reference
    ActorClass:Actor Class Reference=&BP_AutoWireActor_C
   ) -> (then, OutActors:Array of BP Auto Wire Actor Object References)
   > n22 = For Each Loop with Break!(Exec:Exec, Array:Array of BP Auto Wire Actor Object References=@n21, Break:Exec=@n24.then) -> (LoopBody:Exec, Array Element:BP Auto Wire Actor Object Reference, Array Index:int, Completed:Exec)
   Branch(execute=@n22.Completed, Condition=Get FailToFindMatch() -> (FailToFindMatch:bool))
   > Set CurrentSpawnedWire()
   n23 = Branch(execute=@n22.LoopBody, Condition="Equal (Object)"(A=Get KiteBalloonMeshActor(self=@n22), B=Get KiteBallonActor()))
   > Set CurrentSpawnedWire(CurrentSpawnedWire=@n22)
   > n24 = Set FailToFindMatch(FailToFindMatch:bool=false) -> (Output_Get:bool)
   Set FailToFindMatch(execute=@n23.else, FailToFindMatch=true)
   Branch(execute=@n18.else, Condition=IsValid(Object=Get AnchorActor()))
   > n25 = Get()
   > n26 = Branch(Condition="Equal (Object)"(
     A=Get GroundStaticMeshActor!(self:BP Auto Wire Actor Object Reference=@n25) -> (GroundStaticMeshActor:Actor Object Reference)
     B=Get AnchorActor()
    ))
   n27 = GetAllActorsOfClass(execute=[@n26.else, @n25.else], ActorClass=&BP_AutoWireActor_C)
   > n28 = For Each Loop with Break(Array=@n27, Break=@n30.then)
   Branch(execute=@n28.Completed, Condition=Get FailToFindMatch())
   > Set CurrentSpawnedWire()
   n29 = Branch(execute=@n28.LoopBody, Condition="Equal (Object)"(A=Get GroundStaticMeshActor(self=@n28), B=Get AnchorActor()))
   > Set CurrentSpawnedWire(CurrentSpawnedWire=@n28)
   > n30 = Set FailToFindMatch(FailToFindMatch=false)
   Set FailToFindMatch(execute=@n29.else, FailToFindMatch=true)
   n31 = GetDisplayName(Object=@n17)
   n32 = Contains(SearchIn=@n31, Substring=_Inflatable, bSearchFromEnd=true)
   n33 = Branch(execute=@n17.LoopBody, Condition=OR Boolean(A=Contains(SearchIn=@n31, Substring=_Kite, bSearchFromEnd=true), B=@n32))
   > Set KiteBallonActor(KiteBallonActor=@n17)
   > Set bInflatable(bInflatable:bool=@n32) -> (Output_Get:bool)
   Set AnchorActor(execute=@n33.else, AnchorActor=@n17)
   FillComboBox!(
    execute=@n15.then_1
    self
    Actor:Actor Object Reference=Get KiteBallonActor()
    ComboBox:"Searchable Combo Box (String) Object Reference"=Get KiteSocketComboBox() -> (KiteSocketComboBox:"Searchable Combo Box (String) Object Reference")
   ) -> (then)
   > FillComboBox(Actor=Get AnchorActor(), ComboBox=Get GroundSocketComboBox() -> (GroundSocketComboBox:"Searchable Combo Box (String) Object Reference"))
   > SetObject(self=Get RotationProperty() -> (RotationProperty:Single Property View Object Reference), NewObject=Get CurrentSpawnedWire())
   UpdateValues(execute=@n15.then_2)
   n34 = "On Selection Changed (KiteSocketComboBox)"!() -> (OutputDelegate, then, SelectedItem:String, SelectionType:ESelectInfo Enum) {is=title K2Node_ComponentBoundEvent}
   > Set KiteSocketAnchorName(KiteSocketAnchorName:Name=String To Name(InString=@n34.SelectedItem)) -> (Output_Get:Name=None)
   n35 = "On Selection Changed (GroundSocketComboBox)"!() -> (OutputDelegate, then, SelectedItem:String, SelectionType:ESelectInfo Enum) {is=title K2Node_ComponentBoundEvent}
   > Set GroundSocketAnchorName(GroundSocketAnchorName:Name=String To Name(InString=@n35.SelectedItem)) -> (Output_Get:Name=None)
   "On Property Changed (KiteBallonMeshProperty)"!() -> (OutputDelegate, then, PropertyName:Name=None) {is=title K2Node_ComponentBoundEvent}
   > FillComboBox(Actor=Get KiteBallonActor(), ComboBox=Get KiteSocketComboBox())
   "On Property Changed (AnchorActorProperty)"!() -> (OutputDelegate, then, PropertyName:Name=None) {is=title K2Node_ComponentBoundEvent}
   > FillComboBox(Actor=Get AnchorActor(), ComboBox=Get GroundSocketComboBox())
   n36 = "On Clicked (UpdateWireButton)"() {is=title K2Node_ComponentBoundEvent}
   n37 = Get!(execute=[@n36.then, @n41.then]) -> (then, else, KiteData:BP Kite Wire Data Object Reference)
   > n38 = Sequence()
   n39 = GetAllValues!(self:BP Kite Wire Data Object Reference=@n37) -> (WireMaterial:Material Instance Object Reference, Thickness:double)
   > Set WireMaterial(WireMaterial:Material Instance Object Reference=@n39) -> (Output_Get:Material Instance Object Reference)
   Set Thickness(execute=@n38.then_1, Thickness:double=@n39.Thickness) -> (Output_Get:double)
   n40 = "On Property Changed (LengthProperty)"!() -> (OutputDelegate, then, PropertyName:Name=None) {is=title K2Node_ComponentBoundEvent}
   n41 = "On Property Changed (KiteDataProperty)"!() -> (OutputDelegate, then, PropertyName:Name=None) {is=title K2Node_ComponentBoundEvent}
   n42 = SetGrabbedActors() {is=title K2Node_CustomEvent}
   Event PreConstruct()
   > SetObject(
    self=[
     Get RotationProperty()
     Get WireActorProperty() -> (WireActorProperty:Single Property View Object Reference)
     Get WireMeshProperty() -> (WireMeshProperty:Single Property View Object Reference)
     Get LengthProperty() -> (LengthProperty:Single Property View Object Reference)
     Get CurveProperty() -> (CurveProperty:Single Property View Object Reference)
     Get GroundSocketNameProperty() -> (GroundSocketNameProperty:Single Property View Object Reference)
     Get KiteSocketNameProperty() -> (KiteSocketNameProperty:Single Property View Object Reference)
     Get KiteBallonMeshProperty() -> (KiteBallonMeshProperty:Single Property View Object Reference)
     Get ForwardAxisProperty() -> (ForwardAxisProperty:Single Property View Object Reference)
     Get AutoLookAtProperty() -> (AutoLookAtProperty:Single Property View Object Reference)
     Get AnchorActorProperty() -> (AnchorActorProperty:Single Property View Object Reference)
     Get KiteDataProperty() -> (KiteDataProperty:Single Property View Object Reference)
     Get ConnectionAWireProperty() -> (ConnectionAWireProperty:Single Property View Object Reference)
     Get ConnectionBWireProperty() -> (ConnectionBWireProperty:Single Property View Object Reference)
    ]
    NewObject=Self-Reference()
   )
   Event Construct()
   > UpdateValues()
  ]}
 FillComboBox {nodes [
   n1 = FillComboBox!() -> (then, Actor:Actor Object Reference, ComboBox:"Searchable Combo Box (String) Object Reference") {is=title K2Node_FunctionEntry}
   > ClearOptions(self:"Searchable Combo Box (String) Object Reference"=@n1.ComboBox)
   n2 = ConvertSelectedActor!(Target:Actor Object Reference=@n1) -> (ReturnValue:Static Mesh Component Object Reference, IsValid:bool) {is=title K2Node_CallFunction}
   > n3 = Branch(Condition=@n2.IsValid)
   > AddOption!(execute, self:"Searchable Combo Box (String) Object Reference"=@n1.ComboBox, Option:String=Select Socket Name) -> (then)
   > n4 = For Each Loop(Array:Array of Names=GetAllSocketNames!(self:Scene Component Object Reference=@n2) -> (ReturnValue:Array of Names)) -> (Array Element:Name)
   > AddOption(self=@n1.ComboBox, Option="To String (Name)"(InName=@n4))
   ClearOptions(execute=@n3.else, self=@n1.ComboBox)
   > AddOption(self=@n1.ComboBox, Option=None - Empty)
   > SetSelectedOption!(execute, self:"Searchable Combo Box (String) Object Reference"=@n1.ComboBox, Option:String=None - Empty) -> (then)
  ]}
 UpdateValues {nodes [
   UpdateValues() {is=title K2Node_FunctionEntry}
   > n1 = Get!(execute) -> (then, else, CurrentSpawnedWire:BP Auto Wire Actor Object Reference)
   > n2 = Sequence!(execute) -> (then_0, then_1, then_2, then_3:Exec, then_4:Exec, then_5:Exec, then_6:Exec, then_7:Exec, then_8:Exec, then_9:Exec)
   n3 = GetAllValues!(self:BP Auto Wire Actor Object Reference=@n1) -> (
    KiteData:BP Kite Wire Data Object Reference
    KiteBalloonMeshActor:Actor Object Reference
    WireMaterial
    GroundWireConnectionMesh:Static Mesh Object Reference
    Thickness
    AutoLookAt:bool
    KiteSocketAnchorName:Name=None
    GroundSocketAnchorName:Name=None
    GroundStaticMeshActor:Actor Object Reference
    bInflatable:bool
   )
   > Set KiteBallonActor(KiteBallonActor=@n3.KiteBalloonMeshActor)
   Set WireMaterial(execute=@n2.then_1, WireMaterial=@n3.WireMaterial)
   Set AnchorMesh(execute=@n2.then_2, AnchorMesh:Static Mesh Object Reference=@n3.GroundWireConnectionMesh) -> (Output_Get:Static Mesh Object Reference)
   Set Thickness(execute=@n2.then_3, Thickness=@n3.Thickness)
   Set KiteData(execute=@n2.then_4, KiteData:BP Kite Wire Data Object Reference=@n3) -> (Output_Get:BP Kite Wire Data Object Reference)
   Set AutoLookAt(execute=@n2.then_5, AutoLookAt:bool=@n3.AutoLookAt) -> (Output_Get:bool)
   n4 = Set KiteSocketAnchorName(execute=@n2.then_6, KiteSocketAnchorName=@n3.KiteSocketAnchorName) -> (Output_Get=None)
   > SetSelectedOption(self=Get KiteSocketComboBox(), Option="To String (Name)"(InName=@n4))
   n5 = Set GroundSocketAnchorName(execute=@n2.then_7, GroundSocketAnchorName=@n3.GroundSocketAnchorName) -> (Output_Get=None)
   > SetSelectedOption(self=Get GroundSocketComboBox(), Option="To String (Name)"(InName=@n5))
   Set AnchorActor(execute=@n2.then_8, AnchorActor=@n3.GroundStaticMeshActor)
   Set bInflatable(execute=@n2.then_9, bInflatable=@n3.bInflatable)
   n6 = Get!(execute=@n1.else) -> (then, else, KiteData:BP Kite Wire Data Object Reference)
   > n7 = Sequence!(execute) -> (then_0, then_1, then_2)
   n8 = GetAllValues!(self:BP Kite Wire Data Object Reference=@n6) -> (WireMaterial, Thickness)
   > Set WireMaterial(WireMaterial=@n8)
   Set Thickness(execute=@n7.then_1, Thickness=@n8.Thickness)
   SetSelectedOption(execute=[@n7.then_2, @n6.else], self=Get GroundSocketComboBox(), Option=Select Socket Name)
   > SetSelectedOption(self=Get KiteSocketComboBox(), Option=Select Socket Name)
  ]}
]
variables [
 AnchorActor {type=Actor, default=null}
 AnchorMesh {type=StaticMesh, default=/Game/ArtTools/Tech_Art/KiteAutoWireTool/CozySoup_KiteAnchor_A/Meshes/SM_CozySoup_KiteAnchor_A}
 AutoLookAt {type=boolean, default=true}
 bInflatable {type=boolean}
 CurrentSpawnedWire {type=BP_AutoWireActor_C, default=null}
 FailToFindMatch {type=boolean}
 GroundSocketAnchorName {type=string, default=Anchor}
 KiteBallonActor {type=Actor, default=null}
 KiteData {type=BP_KiteWireData_C, default=/Game/ArtTools/Tech_Art/KiteAutoWireTool/Data/DA_KiteDefaultWire}
 KiteSocketAnchorName {type=string, default=FX_Death}
 Rotation {type=number}
 SpilineCurve {type=number, default=35}
 Thickness {type=number, default=3.5}
 WireMaterial {type=MaterialInstance, default=null}
 WireMeshForwardAxis {type=ESplineMeshAxis, default=Z}
]
widgets {
 info {rootWidgetClass=&Overlay, widgetCount=50}
 widgets [
  Overlay_0 {widgetClassPath=&Overlay, widget=Overlay_0}
  EditorUtilityScrollBox_1 {
   widgetClassPath=&EditorUtilityScrollBox
   parent=Overlay_0
   widget=EditorUtilityScrollBox_1
   properties {orientation=Orient_Horizontal}
   slot {class=OverlaySlot, properties {horizontalAlignment=HAlign_Fill, verticalAlignment=VAlign_Fill, padding=<ltrb 10 10 10 10>}}
  }
  EditorUtilityScrollBox_0 {
   widgetClassPath=&EditorUtilityScrollBox
   parent=EditorUtilityScrollBox_1
   widget=EditorUtilityScrollBox_0
   slot {class=ScrollBoxSlot, properties {size {sizeRule=Fill, value=1}}}
  }
  VerticalBox_0 {
   widgetClassPath=&VerticalBox
   parent=EditorUtilityScrollBox_0
   widget=VerticalBox_0
   slot {class=ScrollBoxSlot, properties {size {sizeRule=Fill, value=1}}}
  }
  HorizontalBox_1 {
   widgetClassPath=&HorizontalBox
   parent=VerticalBox_0
   widget=HorizontalBox_1
   slot {class=VerticalBoxSlot, properties {horizontalAlignment=HAlign_Left, verticalAlignment=VAlign_Top, padding=<ltrb 0 0 0 10>}}
  }
  UpdateWireButton {widgetClassPath=&EditorUtilityButton, parent=HorizontalBox_1, widget=UpdateWireButton, slot {class=HorizontalBoxSlot}}
  TextBlock_69 {
   widgetClassPath=&TextBlock
   parent=UpdateWireButton
   widget=TextBlock_69
   properties {
    text=Update Wire
    font {
     bForceMonospaced=false
     bMaterialIsStencil=false
     fontMaterial=null
     fontObject=&Roboto
     letterSpacing=0
     monospacedWidth=1
     size=11.25
     skewAmount=0
     typefaceFontName=Bold
     outlineSettings {
      bApplyOutlineToDropShadows=false
      bMiteredCorners=false
      bSeparateFillAlpha=false
      outlineMaterial=null
      outlineSize=0
      outlineColor=<rgba 0 0 0 1>
     }
    }
   }
   slot {class=ButtonSlot}
  }
  GrabActorButton {widgetClassPath=&EditorUtilityButton, parent=HorizontalBox_1, widget=GrabActorButton, slot {class=HorizontalBoxSlot}}
  TextBlock {
   widgetClassPath=&TextBlock
   parent=GrabActorButton
   widget=TextBlock
   properties {
    text=Grab Selected Actor
    font {
     bForceMonospaced=false
     bMaterialIsStencil=false
     fontMaterial=null
     fontObject=&Roboto
     letterSpacing=0
     monospacedWidth=1
     size=11.25
     skewAmount=0
     typefaceFontName=Bold
     outlineSettings {
      bApplyOutlineToDropShadows=false
      bMiteredCorners=false
      bSeparateFillAlpha=false
      outlineMaterial=null
      outlineSize=0
      outlineColor=<rgba 0 0 0 1>
     }
    }
   }
   slot {class=ButtonSlot}
  }
  Border_9 {
   widgetClassPath=&Border
   parent=VerticalBox_0
   widget=Border_9
   properties {brushColor=<rgba 0.0573 0.0573 0.0573 1>}
   slot {class=VerticalBoxSlot, properties {padding=<ltrb 0 0 0 2>}}
  }
  KiteDataProperty {
   widgetClassPath=&SinglePropertyView
   parent=Border_9
   widget=KiteDataProperty
   properties {nameOverride=Wire Type Data, propertyName=KiteData}
   slot {class=BorderSlot}
  }
  Border_8 {
   widgetClassPath=&Border
   parent=VerticalBox_0
   widget=Border_8
   properties {brushColor=<rgba 0.0573 0.0573 0.0573 1>}
   slot {class=VerticalBoxSlot, properties {padding=<ltrb 0 0 0 2>}}
  }
  WireActorProperty {
   widgetClassPath=&SinglePropertyView
   parent=Border_8
   widget=WireActorProperty
   properties {propertyName=CurrentSpawnedWire}
   slot {class=BorderSlot}
  }
  VerticalBox_1 {widgetClassPath=&VerticalBox, parent=VerticalBox_0, widget=VerticalBox_1, slot {class=VerticalBoxSlot}}
  HorizontalBox_4 {
   widgetClassPath=&HorizontalBox
   parent=VerticalBox_1
   widget=HorizontalBox_4
   slot {class=VerticalBoxSlot, properties {padding=<ltrb 0 0 0 2>}}
  }
  Border_6 {
   widgetClassPath=&Border
   parent=HorizontalBox_4
   widget=Border_6
   properties {clipping=ClipToBounds, brushColor=<rgba 0.0573 0.0573 0.0573 1>}
   slot {class=HorizontalBoxSlot}
  }
  KiteBallonMeshProperty {
   widgetClassPath=&SinglePropertyView
   parent=Border_6
   widget=KiteBallonMeshProperty
   properties {propertyName=KiteBallonActor}
   slot {class=BorderSlot}
  }
  Border_7 {
   widgetClassPath=&Border
   parent=HorizontalBox_4
   widget=Border_7
   properties {brushColor=<rgba 0.0573 0.0573 0.0573 1>}
   slot {class=HorizontalBoxSlot, properties {padding=<ltrb 2 0 0 0>, size {sizeRule=Fill, value=1}}}
  }
  VerticalBox_3 {widgetClassPath=&VerticalBox, parent=Border_7, widget=VerticalBox_3, slot {class=BorderSlot}}
  HorizontalBox_0 {widgetClassPath=&HorizontalBox, parent=VerticalBox_3, widget=HorizontalBox_0, slot {class=VerticalBoxSlot}}
  TextBlock_42 {
   widgetClassPath=&TextBlock
   parent=HorizontalBox_0
   widget=TextBlock_42
   properties {
    text=Socket Name
    colorAndOpacity {colorUseRule=UseColor_Specified, specifiedColor=<rgba 0.5271 0.5271 0.5271 1>}
    font {
     bForceMonospaced=false
     bMaterialIsStencil=false
     fontMaterial=null
     fontObject=&Roboto
     letterSpacing=0
     monospacedWidth=1
     size=7.88
     skewAmount=0
     typefaceFontName=Regular
     outlineSettings {
      bApplyOutlineToDropShadows=false
      bMiteredCorners=false
      bSeparateFillAlpha=false
      outlineMaterial=null
      outlineSize=0
      outlineColor=<rgba 0 0 0 1>
     }
    }
   }
   slot {class=HorizontalBoxSlot, properties {verticalAlignment=VAlign_Center, padding=<ltrb 4 0 10 0>}}
  }
  KiteSocketComboBox {
   widgetClassPath=/Script/FortniteEditor.SearchableStringComboBox
   bIsVariable=true
   parent=HorizontalBox_0
   widget=KiteSocketComboBox
   properties {
    contentPadding=<ltrb 0 0 0 0>
    font {
     bForceMonospaced=false
     bMaterialIsStencil=false
     fontMaterial=null
     fontObject=&Roboto
     letterSpacing=0
     monospacedWidth=1
     size=7.88
     skewAmount=0
     typefaceFontName=Regular
     outlineSettings {
      bApplyOutlineToDropShadows=false
      bMiteredCorners=false
      bSeparateFillAlpha=false
      outlineMaterial=null
      outlineSize=0
      outlineColor=<rgba 0 0 0 1>
     }
    }
    foregroundColor {colorUseRule=UseColor_Specified, specifiedColor=<rgba 0.5271 0.5271 0.5271 1>}
    itemStyle {
     bUseParentRowBrush=false
     activeBrush {
      bIsDynamicallyLoaded=false
      drawAs=Image
      imageType=NoImage
      mirroring=NoMirror
      resourceName=None
      resourceObject=null
      tiling=NoTile
      imageSize=<0 0>
      margin=<ltrb 0 0 0 0>
      outlineSettings {
       bUseBrushTransparency=false
       roundingType=HalfHeightRadius
       width=0
       color {colorUseRule=UseColor_Specified, specifiedColor=<rgba 0 0 0 0>}
       cornerRadii=<0 0 0 0>
      }
      tintColor {colorUseRule=UseColor_Specified, specifiedColor=<rgba 0 0.162 0.7454 1>}
      uVRegion {bIsValid=false, max=<0 0>, min=<0 0>}
     }
     activeHighlightedBrush {
      bIsDynamicallyLoaded=false
      drawAs=Image
      imageType=NoImage
      mirroring=NoMirror
      resourceName=None
      resourceObject=null
      tiling=NoTile
      imageSize=<0 0>
      margin=<ltrb 0 0 0 0>
      outlineSettings {
       bUseBrushTransparency=false
       roundingType=HalfHeightRadius
       width=0
       color {colorUseRule=UseColor_Specified, specifiedColor=<rgba 0 0 0 0>}
       cornerRadii=<0 0 0 0>
      }
      tintColor {colorUseRule=UseColor_Specified, specifiedColor=<rgba 0.0252 0.0319 0.0423 1>}
      uVRegion {bIsValid=false, max=<0 0>, min=<0 0>}
     }
     activeHoveredBrush {
      bIsDynamicallyLoaded=false
      drawAs=Image
      imageType=NoImage
      mirroring=NoMirror
      resourceName=None
      resourceObject=null
      tiling=NoTile
      imageSize=<0 0>
      margin=<ltrb 0 0 0 0>
      outlineSettings {
       bUseBrushTransparency=false
       roundingType=HalfHeightRadius
       width=0
       color {colorUseRule=UseColor_Specified, specifiedColor=<rgba 0 0 0 0>}
       cornerRadii=<0 0 0 0>
      }
      tintColor {colorUseRule=UseColor_Specified, specifiedColor=<rgba 0 0.162 0.7454 1>}
      uVRegion {bIsValid=false, max=<0 0>, min=<0 0>}
     }
     dropIndicator_Above {
      bIsDynamicallyLoaded=false
      drawAs=Box
      imageType=FullColor
      mirroring=NoMirror
      resourceName=../../../Engine/Content/Slate/Common/DropZoneIndicator_Above.png
      resourceObject=null
      tiling=NoTile
      imageSize=<0 0>
      margin=<ltrb 0.625 0.625 0 0>
      outlineSettings {
       bUseBrushTransparency=false
       roundingType=HalfHeightRadius
       width=0
       color {colorUseRule=UseColor_Specified, specifiedColor=<rgba 0 0 0 0>}
       cornerRadii=<0 0 0 0>
      }
      tintColor {colorUseRule=UseColor_Specified, specifiedColor=<rgba 0.0091 0.3515 0.9647 1>}
      uVRegion {bIsValid=false, max=<0 0>, min=<0 0>}
     }
     dropIndicator_Below {
      bIsDynamicallyLoaded=false
      drawAs=Box
      imageType=FullColor
      mirroring=NoMirror
      resourceName=../../../Engine/Content/Slate/Common/DropZoneIndicator_Below.png
      resourceObject=null
      tiling=NoTile
      imageSize=<0 0>
      margin=<ltrb 0.625 0 0 0.625>
      outlineSettings {
       bUseBrushTransparency=false
       roundingType=HalfHeightRadius
       width=0
       color {colorUseRule=UseColor_Specified, specifiedColor=<rgba 0 0 0 0>}
       cornerRadii=<0 0 0 0>
      }
      tintColor {colorUseRule=UseColor_Specified, specifiedColor=<rgba 0.0091 0.3515 0.9647 1>}
      uVRegion {bIsValid=false, max=<0 0>, min=<0 0>}
     }
     dropIndicator_Onto {
      bIsDynamicallyLoaded=false
      drawAs=Box
      imageType=FullColor
      mirroring=NoMirror
      resourceName=../../../Engine/Content/Slate/Common/DropZoneIndicator_Onto.png
      resourceObject=null
      tiling=NoTile
      imageSize=<0 0>
      margin=<ltrb 0.25 0.25 0.25 0.25>
      outlineSettings {
       bUseBrushTransparency=false
       roundingType=HalfHeightRadius
       width=0
       color {colorUseRule=UseColor_Specified, specifiedColor=<rgba 0 0 0 0>}
       cornerRadii=<0 0 0 0>
      }
      tintColor {colorUseRule=UseColor_Specified, specifiedColor=<rgba 0.0091 0.3515 0.9647 1>}
      uVRegion {bIsValid=false, max=<0 0>, min=<0 0>}
     }
     evenRowBackgroundBrush {
      bIsDynamicallyLoaded=false
      drawAs=Image
      imageType=NoImage
      mirroring=NoMirror
      resourceName=None
      resourceObject=null
      tiling=NoTile
      imageSize=<0 0>
      margin=<ltrb 0 0 0 0>
      outlineSettings {
       bUseBrushTransparency=false
       roundingType=HalfHeightRadius
       width=0
       color {colorUseRule=UseColor_Specified, specifiedColor=<rgba 0 0 0 0>}
       cornerRadii=<0 0 0 0>
      }
      tintColor {colorUseRule=UseColor_Specified, specifiedColor=<rgba 0.0103 0.0103 0.0103 1>}
      uVRegion {bIsValid=false, max=<0 0>, min=<0 0>}
     }
     evenRowBackgroundHoveredBrush {
      bIsDynamicallyLoaded=false
      drawAs=Image
      imageType=NoImage
      mirroring=NoMirror
      resourceName=None
      resourceObject=null
      tiling=NoTile
      imageSize=<0 0>
      margin=<ltrb 0 0 0 0>
      outlineSettings {
       bUseBrushTransparency=false
       roundingType=HalfHeightRadius
       width=0
       color {colorUseRule=UseColor_Specified, specifiedColor=<rgba 0 0 0 0>}
       cornerRadii=<0 0 0 0>
      }
      tintColor {colorUseRule=UseColor_Specified, specifiedColor=<rgba 0.0176 0.0176 0.0176 1>}
      uVRegion {bIsValid=false, max=<0 0>, min=<0 0>}
     }
     inactiveBrush {
      bIsDynamicallyLoaded=false
      drawAs=Image
      imageType=NoImage
      mirroring=NoMirror
      resourceName=None
      resourceObject=null
      tiling=NoTile
      imageSize=<0 0>
      margin=<ltrb 0 0 0 0>
      outlineSettings {
       bUseBrushTransparency=false
       roundingType=HalfHeightRadius
       width=0
       color {colorUseRule=UseColor_Specified, specifiedColor=<rgba 0 0 0 0>}
       cornerRadii=<0 0 0 0>
      }
      tintColor {colorUseRule=UseColor_Specified, specifiedColor=<rgba 0.0513 0.0953 0.159 1>}
      uVRegion {bIsValid=false, max=<0 0>, min=<0 0>}
     }
     inactiveHighlightedBrush {
      bIsDynamicallyLoaded=false
      drawAs=Image
      imageType=NoImage
      mirroring=NoMirror
      resourceName=None
      resourceObject=null
      tiling=NoTile
      imageSize=<0 0>
      margin=<ltrb 0 0 0 0>
      outlineSettings {
       bUseBrushTransparency=false
       roundingType=HalfHeightRadius
       width=0
       color {colorUseRule=UseColor_Specified, specifiedColor=<rgba 0 0 0 0>}
       cornerRadii=<0 0 0 0>
      }
      tintColor {colorUseRule=UseColor_Specified, specifiedColor=<rgba 0.0252 0.0319 0.0423 1>}
      uVRegion {bIsValid=false, max=<0 0>, min=<0 0>}
     }
     inactiveHoveredBrush {
      bIsDynamicallyLoaded=false
      drawAs=Image
      imageType=NoImage
      mirroring=NoMirror
      resourceName=None
      resourceObject=null
      tiling=NoTile
      imageSize=<0 0>
      margin=<ltrb 0 0 0 0>
      outlineSettings {
       bUseBrushTransparency=false
       roundingType=HalfHeightRadius
       width=0
       color {colorUseRule=UseColor_Specified, specifiedColor=<rgba 0 0 0 0>}
       cornerRadii=<0 0 0 0>
      }
      tintColor {colorUseRule=UseColor_Specified, specifiedColor=<rgba 0.0513 0.0953 0.159 1>}
      uVRegion {bIsValid=false, max=<0 0>, min=<0 0>}
     }
     oddRowBackgroundBrush {
      bIsDynamicallyLoaded=false
      drawAs=Image
      imageType=NoImage
      mirroring=NoMirror
      resourceName=None
      resourceObject=null
      tiling=NoTile
      imageSize=<0 0>
      margin=<ltrb 0 0 0 0>
      outlineSettings {
       bUseBrushTransparency=false
       roundingType=HalfHeightRadius
       width=0
       color {colorUseRule=UseColor_Specified, specifiedColor=<rgba 0 0 0 0>}
       cornerRadii=<0 0 0 0>
      }
      tintColor {colorUseRule=UseColor_Specified, specifiedColor=<rgba 0.0103 0.0103 0.0103 1>}
      uVRegion {bIsValid=false, max=<0 0>, min=<0 0>}
     }
     oddRowBackgroundHoveredBrush {
      bIsDynamicallyLoaded=false
      drawAs=Image
      imageType=NoImage
      mirroring=NoMirror
      resourceName=None
      resourceObject=null
      tiling=NoTile
      imageSize=<0 0>
      margin=<ltrb 0 0 0 0>
      outlineSettings {
       bUseBrushTransparency=false
       roundingType=HalfHeightRadius
       width=0
       color {colorUseRule=UseColor_Specified, specifiedColor=<rgba 0 0 0 0>}
       cornerRadii=<0 0 0 0>
      }
      tintColor {colorUseRule=UseColor_Specified, specifiedColor=<rgba 0.0176 0.0176 0.0176 1>}
      uVRegion {bIsValid=false, max=<0 0>, min=<0 0>}
     }
     parentRowBackgroundBrush {
      bIsDynamicallyLoaded=false
      drawAs=Image
      imageType=NoImage
      mirroring=NoMirror
      resourceName=None
      resourceObject=null
      tiling=NoTile
      imageSize=<32 32>
      margin=<ltrb 0 0 0 0>
      outlineSettings {
       bUseBrushTransparency=false
       roundingType=HalfHeightRadius
       width=0
       color {colorUseRule=UseColor_Specified, specifiedColor=<rgba 0 0 0 0>}
       cornerRadii=<0 0 0 0>
      }
      tintColor {colorUseRule=UseColor_Specified, specifiedColor=<rgba 1 1 1 1>}
      uVRegion {bIsValid=false, max=<0 0>, min=<0 0>}
     }
     parentRowBackgroundHoveredBrush {
      bIsDynamicallyLoaded=false
      drawAs=Image
      imageType=NoImage
      mirroring=NoMirror
      resourceName=None
      resourceObject=null
      tiling=NoTile
      imageSize=<32 32>
      margin=<ltrb 0 0 0 0>
      outlineSettings {
       bUseBrushTransparency=false
       roundingType=HalfHeightRadius
       width=0
       color {colorUseRule=UseColor_Specified, specifiedColor=<rgba 0 0 0 0>}
       cornerRadii=<0 0 0 0>
      }
      tintColor {colorUseRule=UseColor_Specified, specifiedColor=<rgba 1 1 1 1>}
      uVRegion {bIsValid=false, max=<0 0>, min=<0 0>}
     }
     selectedTextColor {colorUseRule=UseColor_Specified, specifiedColor=<rgba 1 1 1 1>}
     selectorFocusedBrush {
      bIsDynamicallyLoaded=false
      drawAs=Border
      imageType=FullColor
      mirroring=NoMirror
      resourceName=../../../Engine/Content/Slate/Common/Selector.png
      resourceObject=null
      tiling=Both
      imageSize=<0 0>
      margin=<ltrb 0.25 0.25 0.25 0.25>
      outlineSettings {
       bUseBrushTransparency=false
       roundingType=HalfHeightRadius
       width=0
       color {colorUseRule=UseColor_Specified, specifiedColor=<rgba 0 0 0 0>}
       cornerRadii=<0 0 0 0>
      }
      tintColor {colorUseRule=UseColor_Specified, specifiedColor=<rgba 0.701 0.225 0.003 1>}
      uVRegion {bIsValid=false, max=<0 0>, min=<0 0>}
     }
     textColor {colorUseRule=UseColor_Specified, specifiedColor=<rgba 0.5271 0.5271 0.5271 1>}
    }
    widgetStyle {
     comboButtonStyle {
      downArrowAlign=VAlign_Center
      buttonStyle {
       clickedSlateSound {resourceObject=null}
       disabled {
        bIsDynamicallyLoaded=false
        drawAs=RoundedBox
        imageType=NoImage
        mirroring=NoMirror
        resourceName=None
        resourceObject=null
        tiling=NoTile
        imageSize=<0 0>
        margin=<ltrb 0 0 0 0>
        outlineSettings {
         bUseBrushTransparency=false
         roundingType=FixedRadius
         width=1
         color {colorUseRule=UseColor_ColorTable, specifiedColor=<rgba 0 0 0 0>}
         cornerRadii=<4 4 4 4>
        }
        tintColor {colorUseRule=UseColor_Specified, specifiedColor=<rgba 0.0395 0.0395 0.0395 1>}
        uVRegion {bIsValid=false, max=<0 0>, min=<0 0>}
       }
       disabledForeground {colorUseRule=UseColor_ColorTable, specifiedColor=<rgba 0 0 0 0>}
       hovered {
        bIsDynamicallyLoaded=false
        drawAs=RoundedBox
        imageType=NoImage
        mirroring=NoMirror
        resourceName=None
        resourceObject=null
        tiling=NoTile
        imageSize=<0 0>
        margin=<ltrb 0 0 0 0>
        outlineSettings {
         bUseBrushTransparency=false
         roundingType=FixedRadius
         width=1
         color {colorUseRule=UseColor_ColorTable, specifiedColor=<rgba 0 0 0 0>}
         cornerRadii=<4 4 4 4>
        }
        tintColor {colorUseRule=UseColor_Specified, specifiedColor=<rgba 0.0048 0.0048 0.0048 1>}
        uVRegion {bIsValid=false, max=<0 0>, min=<0 0>}
       }
       hoveredForeground {colorUseRule=UseColor_ColorTable, specifiedColor=<rgba 0 0 0 0>}
       hoveredSlateSound {resourceObject=null}
       normal {
        bIsDynamicallyLoaded=false
        drawAs=RoundedBox
        imageType=NoImage
        mirroring=NoMirror
        resourceName=None
        resourceObject=null
        tiling=NoTile
        imageSize=<0 0>
        margin=<ltrb 0 0 0 0>
        outlineSettings {
         bUseBrushTransparency=false
         roundingType=FixedRadius
         width=1
         color {colorUseRule=UseColor_ColorTable, specifiedColor=<rgba 0 0 0 0>}
         cornerRadii=<4 4 4 4>
        }
        tintColor {colorUseRule=UseColor_Specified, specifiedColor=<rgba 0.0048 0.0048 0.0048 1>}
        uVRegion {bIsValid=false, max=<0 0>, min=<0 0>}
       }
       normalForeground {colorUseRule=UseColor_ColorTable, specifiedColor=<rgba 0 0 0 0>}
       normalPadding=<ltrb 8 1 8 1>
       pressed {
        bIsDynamicallyLoaded=false
        drawAs=RoundedBox
        imageType=NoImage
        mirroring=NoMirror
        resourceName=None
        resourceObject=null
        tiling=NoTile
        imageSize=<0 0>
        margin=<ltrb 0 0 0 0>
        outlineSettings {
         bUseBrushTransparency=false
         roundingType=FixedRadius
         width=1
         color {colorUseRule=UseColor_ColorTable, specifiedColor=<rgba 0 0 0 0>}
         cornerRadii=<4 4 4 4>
        }
        tintColor {colorUseRule=UseColor_Specified, specifiedColor=<rgba 0.0048 0.0048 0.0048 1>}
        uVRegion {bIsValid=false, max=<0 0>, min=<0 0>}
       }
       pressedForeground {colorUseRule=UseColor_ColorTable, specifiedColor=<rgba 0 0 0 0>}
       pressedPadding=<ltrb 8 2 8 0>
       pressedSlateSound {resourceObject=null}
      }
      contentPadding=<ltrb 0 0 0 0>
      downArrowImage {
       bIsDynamicallyLoaded=false
       drawAs=Image
       imageType=Vector
       mirroring=NoMirror
       resourceName=../../../Engine/Content/Slate/Starship/CoreWidgets/ComboBox/wide-chevron-down.svg
       resourceObject=null
       tiling=NoTile
       imageSize=<16 16>
       margin=<ltrb 0 0 0 0>
       outlineSettings {
        bUseBrushTransparency=false
        roundingType=HalfHeightRadius
        width=0
        color {colorUseRule=UseColor_Specified, specifiedColor=<rgba 0 0 0 0>}
        cornerRadii=<0 0 0 0>
       }
       tintColor {colorUseRule=UseColor_Specified, specifiedColor=<rgba 1 1 1 1>}
       uVRegion {bIsValid=false, max=<0 0>, min=<0 0>}
      }
      downArrowPadding=<ltrb 2 2 2 2>
      menuBorderBrush {
       bIsDynamicallyLoaded=false
       drawAs=RoundedBox
       imageType=NoImage
       mirroring=NoMirror
       resourceName=None
       resourceObject=null
       tiling=NoTile
       imageSize=<0 0>
       margin=<ltrb 0 0 0 0>
       outlineSettings {
        bUseBrushTransparency=false
        roundingType=FixedRadius
        width=1
        color {colorUseRule=UseColor_ColorTable, specifiedColor=<rgba 0 0 0 0>}
        cornerRadii=<0 0 0 0>
       }
       tintColor {colorUseRule=UseColor_Specified, specifiedColor=<rgba 0.0395 0.0395 0.0395 1>}
       uVRegion {bIsValid=false, max=<0 0>, min=<0 0>}
      }
      menuBorderPadding=<ltrb 0 0 0 0>
      shadowColorAndOpacity=<rgba 0 0 0 1>
      shadowOffset=<0 0>
     }
     contentPadding=<ltrb 0 0 0 0>
     menuRowPadding=<ltrb 8 3 8 3>
     pressedSlateSound {resourceObject=null}
     selectionChangeSlateSound {resourceObject=null}
    }
   }
   slot {class=HorizontalBoxSlot, properties {padding=<ltrb 0 0 10 0>, size {sizeRule=Fill, value=1}}}
  }
  KiteSocketNameProperty {
   widgetClassPath=&SinglePropertyView
   parent=VerticalBox_3
   widget=KiteSocketNameProperty
   properties {nameOverride=Socket Name, propertyName=KiteSocketAnchorName, visibility=Collapsed}
   slot {class=VerticalBoxSlot}
  }
  AutoLookAtProperty {
   widgetClassPath=&SinglePropertyView
   parent=VerticalBox_3
   widget=AutoLookAtProperty
   properties {propertyName=AutoLookAt}
   slot {class=VerticalBoxSlot}
  }
  RotationProperty {
   widgetClassPath=&SinglePropertyView
   parent=VerticalBox_3
   widget=RotationProperty
   properties {propertyName=Rotation}
   slot {class=VerticalBoxSlot, properties {horizontalAlignment=HAlign_Left}}
  }
  HorizontalBox_3 {
   widgetClassPath=&HorizontalBox
   parent=VerticalBox_1
   widget=HorizontalBox_3
   properties {visibility=Collapsed}
   slot {class=VerticalBoxSlot, properties {padding=<ltrb 0 0 0 2>}}
  }
  Border_0 {
   widgetClassPath=&Border
   parent=HorizontalBox_3
   widget=Border_0
   properties {brushColor=<rgba 0.0573 0.0573 0.0573 1>}
   slot {class=HorizontalBoxSlot, properties {size {sizeRule=Fill, value=1}}}
  }
  ConnectionAWireProperty {
   widgetClassPath=&SinglePropertyView
   parent=Border_0
   widget=ConnectionAWireProperty
   properties {nameOverride=Connector A, propertyName=ConnectionAWireMesh}
   slot {class=BorderSlot}
  }
  Spacer_142 {
   widgetClassPath=&Spacer
   parent=HorizontalBox_3
   widget=Spacer_142
   properties {visibility=Collapsed, size=<160 1>}
   slot {class=HorizontalBoxSlot}
  }
  HorizontalBox_2 {
   widgetClassPath=&HorizontalBox
   parent=VerticalBox_1
   widget=HorizontalBox_2
   slot {class=VerticalBoxSlot, properties {padding=<ltrb 0 0 0 2>}}
  }
  Border_1 {
   widgetClassPath=&Border
   parent=HorizontalBox_2
   widget=Border_1
   properties {brushColor=<rgba 0.0573 0.0573 0.0573 1>}
   slot {class=HorizontalBoxSlot, properties {size {sizeRule=Fill, value=1}}}
  }
  WireMeshProperty {
   widgetClassPath=&SinglePropertyView
   parent=Border_1
   widget=WireMeshProperty
   properties {bIsEnabled=false, propertyName=WireMaterial}
   slot {class=BorderSlot}
  }
  Border_2 {
   widgetClassPath=&Border
   parent=HorizontalBox_2
   widget=Border_2
   properties {verticalAlignment=VAlign_Center, brushColor=<rgba 0.0573 0.0573 0.0573 1>, padding=<ltrb 0 2 4 2>}
   slot {class=HorizontalBoxSlot, properties {padding=<ltrb 2 0 0 0>}}
  }
  VerticalBox_2 {
   widgetClassPath=&VerticalBox
   parent=Border_2
   widget=VerticalBox_2
   slot {class=BorderSlot, properties {verticalAlignment=VAlign_Center, padding=<ltrb 0 2 4 2>}}
  }
  ForwardAxisProperty {
   widgetClassPath=&SinglePropertyView
   parent=VerticalBox_2
   widget=ForwardAxisProperty
   properties {nameOverride=Mesh Forward Axis, propertyName=WireMeshForwardAxis, visibility=Collapsed}
   slot {class=VerticalBoxSlot, properties {horizontalAlignment=HAlign_Right}}
  }
  LengthProperty {
   widgetClassPath=&SinglePropertyView
   parent=VerticalBox_2
   widget=LengthProperty
   properties {bIsEnabled=false, propertyName=Thickness}
   slot {class=VerticalBoxSlot, properties {horizontalAlignment=HAlign_Right}}
  }
  CurveProperty {
   widgetClassPath=&SinglePropertyView
   parent=VerticalBox_2
   widget=CurveProperty
   properties {propertyName=SpilineCurve, visibility=Collapsed}
   slot {class=VerticalBoxSlot, properties {horizontalAlignment=HAlign_Right}}
  }
  HorizontalBox_6 {
   widgetClassPath=&HorizontalBox
   parent=VerticalBox_1
   widget=HorizontalBox_6
   slot {class=VerticalBoxSlot, properties {padding=<ltrb 0 0 0 2>}}
  }
  Border_3 {
   widgetClassPath=&Border
   parent=HorizontalBox_6
   widget=Border_3
   properties {brushColor=<rgba 0.0573 0.0573 0.0573 1>}
   slot {class=HorizontalBoxSlot, properties {size {sizeRule=Fill, value=1}}}
  }
  ConnectionBWireProperty {
   widgetClassPath=&SinglePropertyView
   parent=Border_3
   widget=ConnectionBWireProperty
   properties {nameOverride=Ground Connection Point Mesh, propertyName=AnchorMesh}
   slot {class=BorderSlot}
  }
  Spacer {
   widgetClassPath=&Spacer
   parent=HorizontalBox_6
   widget=Spacer
   properties {visibility=Collapsed, size=<160 1>}
   slot {class=HorizontalBoxSlot}
  }
  HorizontalBox_5 {
   widgetClassPath=&HorizontalBox
   parent=VerticalBox_1
   widget=HorizontalBox_5
   slot {class=VerticalBoxSlot, properties {padding=<ltrb 0 0 0 2>}}
  }
  Border_4 {
   widgetClassPath=&Border
   parent=HorizontalBox_5
   widget=Border_4
   properties {brushColor=<rgba 0.0573 0.0573 0.0573 1>}
   slot {class=HorizontalBoxSlot}
  }
  AnchorActorProperty {
   widgetClassPath=&SinglePropertyView
   parent=Border_4
   widget=AnchorActorProperty
   properties {propertyName=AnchorActor}
   slot {class=BorderSlot}
  }
  Border_5 {
   widgetClassPath=&Border
   parent=HorizontalBox_5
   widget=Border_5
   properties {visibility=Collapsed, brushColor=<rgba 0.0573 0.0573 0.0573 1>}
   slot {class=HorizontalBoxSlot, properties {padding=<ltrb 2 0 0 0>, size {sizeRule=Fill, value=1}}}
  }
  GroundSocketNameProperty {
   widgetClassPath=&SinglePropertyView
   parent=Border_5
   widget=GroundSocketNameProperty
   properties {nameOverride=Socket Name, propertyName=GroundSocketAnchorName}
   slot {class=BorderSlot}
  }
  Border {
   widgetClassPath=&Border
   parent=HorizontalBox_5
   widget=Border
   properties {brushColor=<rgba 0.0573 0.0573 0.0573 1>, padding=<ltrb 0 0 0 0>}
   slot {class=HorizontalBoxSlot, properties {padding=<ltrb 2 0 0 0>, size {sizeRule=Fill, value=1}}}
  }
  HorizontalBox {widgetClassPath=&HorizontalBox, parent=Border, widget=HorizontalBox, slot {class=BorderSlot, properties {padding=<ltrb 0 0 0 0>}}}
  TextBlock_1 {
   widgetClassPath=&TextBlock
   parent=HorizontalBox
   widget=TextBlock_1
   properties {
    text=Socket Name
    colorAndOpacity {colorUseRule=UseColor_Specified, specifiedColor=<rgba 0.5271 0.5271 0.5271 1>}
    font {
     bForceMonospaced=false
     bMaterialIsStencil=false
     fontMaterial=null
     fontObject=&Roboto
     letterSpacing=0
     monospacedWidth=1
     size=7.88
     skewAmount=0
     typefaceFontName=Regular
     outlineSettings {
      bApplyOutlineToDropShadows=false
      bMiteredCorners=false
      bSeparateFillAlpha=false
      outlineMaterial=null
      outlineSize=0
      outlineColor=<rgba 0 0 0 1>
     }
    }
   }
   slot {class=HorizontalBoxSlot, properties {verticalAlignment=VAlign_Center, padding=<ltrb 4 0 10 0>}}
  }
  GroundSocketComboBox {
   widgetClassPath=&SearchableStringComboBox
   bIsVariable=true
   parent=HorizontalBox
   widget=GroundSocketComboBox
   properties {
    contentPadding=<ltrb 0 0 0 0>
    font {
     bForceMonospaced=false
     bMaterialIsStencil=false
     fontMaterial=null
     fontObject=&Roboto
     letterSpacing=0
     monospacedWidth=1
     size=7.88
     skewAmount=0
     typefaceFontName=Regular
     outlineSettings {
      bApplyOutlineToDropShadows=false
      bMiteredCorners=false
      bSeparateFillAlpha=false
      outlineMaterial=null
      outlineSize=0
      outlineColor=<rgba 0 0 0 1>
     }
    }
    foregroundColor {colorUseRule=UseColor_Specified, specifiedColor=<rgba 0.5271 0.5271 0.5271 1>}
    itemStyle {
     bUseParentRowBrush=false
     activeBrush {
      bIsDynamicallyLoaded=false
      drawAs=Image
      imageType=NoImage
      mirroring=NoMirror
      resourceName=None
      resourceObject=null
      tiling=NoTile
      imageSize=<0 0>
      margin=<ltrb 0 0 0 0>
      outlineSettings {
       bUseBrushTransparency=false
       roundingType=HalfHeightRadius
       width=0
       color {colorUseRule=UseColor_Specified, specifiedColor=<rgba 0 0 0 0>}
       cornerRadii=<0 0 0 0>
      }
      tintColor {colorUseRule=UseColor_Specified, specifiedColor=<rgba 0 0.162 0.7454 1>}
      uVRegion {bIsValid=false, max=<0 0>, min=<0 0>}
     }
     activeHighlightedBrush {
      bIsDynamicallyLoaded=false
      drawAs=Image
      imageType=NoImage
      mirroring=NoMirror
      resourceName=None
      resourceObject=null
      tiling=NoTile
      imageSize=<0 0>
      margin=<ltrb 0 0 0 0>
      outlineSettings {
       bUseBrushTransparency=false
       roundingType=HalfHeightRadius
       width=0
       color {colorUseRule=UseColor_Specified, specifiedColor=<rgba 0 0 0 0>}
       cornerRadii=<0 0 0 0>
      }
      tintColor {colorUseRule=UseColor_Specified, specifiedColor=<rgba 0.0252 0.0319 0.0423 1>}
      uVRegion {bIsValid=false, max=<0 0>, min=<0 0>}
     }
     activeHoveredBrush {
      bIsDynamicallyLoaded=false
      drawAs=Image
      imageType=NoImage
      mirroring=NoMirror
      resourceName=None
      resourceObject=null
      tiling=NoTile
      imageSize=<0 0>
      margin=<ltrb 0 0 0 0>
      outlineSettings {
       bUseBrushTransparency=false
       roundingType=HalfHeightRadius
       width=0
       color {colorUseRule=UseColor_Specified, specifiedColor=<rgba 0 0 0 0>}
       cornerRadii=<0 0 0 0>
      }
      tintColor {colorUseRule=UseColor_Specified, specifiedColor=<rgba 0 0.162 0.7454 1>}
      uVRegion {bIsValid=false, max=<0 0>, min=<0 0>}
     }
     dropIndicator_Above {
      bIsDynamicallyLoaded=false
      drawAs=Box
      imageType=FullColor
      mirroring=NoMirror
      resourceName=../../../Engine/Content/Slate/Common/DropZoneIndicator_Above.png
      resourceObject=null
      tiling=NoTile
      imageSize=<0 0>
      margin=<ltrb 0.625 0.625 0 0>
      outlineSettings {
       bUseBrushTransparency=false
       roundingType=HalfHeightRadius
       width=0
       color {colorUseRule=UseColor_Specified, specifiedColor=<rgba 0 0 0 0>}
       cornerRadii=<0 0 0 0>
      }
      tintColor {colorUseRule=UseColor_Specified, specifiedColor=<rgba 0.0091 0.3515 0.9647 1>}
      uVRegion {bIsValid=false, max=<0 0>, min=<0 0>}
     }
     dropIndicator_Below {
      bIsDynamicallyLoaded=false
      drawAs=Box
      imageType=FullColor
      mirroring=NoMirror
      resourceName=../../../Engine/Content/Slate/Common/DropZoneIndicator_Below.png
      resourceObject=null
      tiling=NoTile
      imageSize=<0 0>
      margin=<ltrb 0.625 0 0 0.625>
      outlineSettings {
       bUseBrushTransparency=false
       roundingType=HalfHeightRadius
       width=0
       color {colorUseRule=UseColor_Specified, specifiedColor=<rgba 0 0 0 0>}
       cornerRadii=<0 0 0 0>
      }
      tintColor {colorUseRule=UseColor_Specified, specifiedColor=<rgba 0.0091 0.3515 0.9647 1>}
      uVRegion {bIsValid=false, max=<0 0>, min=<0 0>}
     }
     dropIndicator_Onto {
      bIsDynamicallyLoaded=false
      drawAs=Box
      imageType=FullColor
      mirroring=NoMirror
      resourceName=../../../Engine/Content/Slate/Common/DropZoneIndicator_Onto.png
      resourceObject=null
      tiling=NoTile
      imageSize=<0 0>
      margin=<ltrb 0.25 0.25 0.25 0.25>
      outlineSettings {
       bUseBrushTransparency=false
       roundingType=HalfHeightRadius
       width=0
       color {colorUseRule=UseColor_Specified, specifiedColor=<rgba 0 0 0 0>}
       cornerRadii=<0 0 0 0>
      }
      tintColor {colorUseRule=UseColor_Specified, specifiedColor=<rgba 0.0091 0.3515 0.9647 1>}
      uVRegion {bIsValid=false, max=<0 0>, min=<0 0>}
     }
     evenRowBackgroundBrush {
      bIsDynamicallyLoaded=false
      drawAs=Image
      imageType=NoImage
      mirroring=NoMirror
      resourceName=None
      resourceObject=null
      tiling=NoTile
      imageSize=<0 0>
      margin=<ltrb 0 0 0 0>
      outlineSettings {
       bUseBrushTransparency=false
       roundingType=HalfHeightRadius
       width=0
       color {colorUseRule=UseColor_Specified, specifiedColor=<rgba 0 0 0 0>}
       cornerRadii=<0 0 0 0>
      }
      tintColor {colorUseRule=UseColor_Specified, specifiedColor=<rgba 0.0103 0.0103 0.0103 1>}
      uVRegion {bIsValid=false, max=<0 0>, min=<0 0>}
     }
     evenRowBackgroundHoveredBrush {
      bIsDynamicallyLoaded=false
      drawAs=Image
      imageType=NoImage
      mirroring=NoMirror
      resourceName=None
      resourceObject=null
      tiling=NoTile
      imageSize=<0 0>
      margin=<ltrb 0 0 0 0>
      outlineSettings {
       bUseBrushTransparency=false
       roundingType=HalfHeightRadius
       width=0
       color {colorUseRule=UseColor_Specified, specifiedColor=<rgba 0 0 0 0>}
       cornerRadii=<0 0 0 0>
      }
      tintColor {colorUseRule=UseColor_Specified, specifiedColor=<rgba 0.0176 0.0176 0.0176 1>}
      uVRegion {bIsValid=false, max=<0 0>, min=<0 0>}
     }
     inactiveBrush {
      bIsDynamicallyLoaded=false
      drawAs=Image
      imageType=NoImage
      mirroring=NoMirror
      resourceName=None
      resourceObject=null
      tiling=NoTile
      imageSize=<0 0>
      margin=<ltrb 0 0 0 0>
      outlineSettings {
       bUseBrushTransparency=false
       roundingType=HalfHeightRadius
       width=0
       color {colorUseRule=UseColor_Specified, specifiedColor=<rgba 0 0 0 0>}
       cornerRadii=<0 0 0 0>
      }
      tintColor {colorUseRule=UseColor_Specified, specifiedColor=<rgba 0.0513 0.0953 0.159 1>}
      uVRegion {bIsValid=false, max=<0 0>, min=<0 0>}
     }
     inactiveHighlightedBrush {
      bIsDynamicallyLoaded=false
      drawAs=Image
      imageType=NoImage
      mirroring=NoMirror
      resourceName=None
      resourceObject=null
      tiling=NoTile
      imageSize=<0 0>
      margin=<ltrb 0 0 0 0>
      outlineSettings {
       bUseBrushTransparency=false
       roundingType=HalfHeightRadius
       width=0
       color {colorUseRule=UseColor_Specified, specifiedColor=<rgba 0 0 0 0>}
       cornerRadii=<0 0 0 0>
      }
      tintColor {colorUseRule=UseColor_Specified, specifiedColor=<rgba 0.0252 0.0319 0.0423 1>}
      uVRegion {bIsValid=false, max=<0 0>, min=<0 0>}
     }
     inactiveHoveredBrush {
      bIsDynamicallyLoaded=false
      drawAs=Image
      imageType=NoImage
      mirroring=NoMirror
      resourceName=None
      resourceObject=null
      tiling=NoTile
      imageSize=<0 0>
      margin=<ltrb 0 0 0 0>
      outlineSettings {
       bUseBrushTransparency=false
       roundingType=HalfHeightRadius
       width=0
       color {colorUseRule=UseColor_Specified, specifiedColor=<rgba 0 0 0 0>}
       cornerRadii=<0 0 0 0>
      }
      tintColor {colorUseRule=UseColor_Specified, specifiedColor=<rgba 0.0513 0.0953 0.159 1>}
      uVRegion {bIsValid=false, max=<0 0>, min=<0 0>}
     }
     oddRowBackgroundBrush {
      bIsDynamicallyLoaded=false
      drawAs=Image
      imageType=NoImage
      mirroring=NoMirror
      resourceName=None
      resourceObject=null
      tiling=NoTile
      imageSize=<0 0>
      margin=<ltrb 0 0 0 0>
      outlineSettings {
       bUseBrushTransparency=false
       roundingType=HalfHeightRadius
       width=0
       color {colorUseRule=UseColor_Specified, specifiedColor=<rgba 0 0 0 0>}
       cornerRadii=<0 0 0 0>
      }
      tintColor {colorUseRule=UseColor_Specified, specifiedColor=<rgba 0.0103 0.0103 0.0103 1>}
      uVRegion {bIsValid=false, max=<0 0>, min=<0 0>}
     }
     oddRowBackgroundHoveredBrush {
      bIsDynamicallyLoaded=false
      drawAs=Image
      imageType=NoImage
      mirroring=NoMirror
      resourceName=None
      resourceObject=null
      tiling=NoTile
      imageSize=<0 0>
      margin=<ltrb 0 0 0 0>
      outlineSettings {
       bUseBrushTransparency=false
       roundingType=HalfHeightRadius
       width=0
       color {colorUseRule=UseColor_Specified, specifiedColor=<rgba 0 0 0 0>}
       cornerRadii=<0 0 0 0>
      }
      tintColor {colorUseRule=UseColor_Specified, specifiedColor=<rgba 0.0176 0.0176 0.0176 1>}
      uVRegion {bIsValid=false, max=<0 0>, min=<0 0>}
     }
     parentRowBackgroundBrush {
      bIsDynamicallyLoaded=false
      drawAs=Image
      imageType=NoImage
      mirroring=NoMirror
      resourceName=None
      resourceObject=null
      tiling=NoTile
      imageSize=<32 32>
      margin=<ltrb 0 0 0 0>
      outlineSettings {
       bUseBrushTransparency=false
       roundingType=HalfHeightRadius
       width=0
       color {colorUseRule=UseColor_Specified, specifiedColor=<rgba 0 0 0 0>}
       cornerRadii=<0 0 0 0>
      }
      tintColor {colorUseRule=UseColor_Specified, specifiedColor=<rgba 1 1 1 1>}
      uVRegion {bIsValid=false, max=<0 0>, min=<0 0>}
     }
     parentRowBackgroundHoveredBrush {
      bIsDynamicallyLoaded=false
      drawAs=Image
      imageType=NoImage
      mirroring=NoMirror
      resourceName=None
      resourceObject=null
      tiling=NoTile
      imageSize=<32 32>
      margin=<ltrb 0 0 0 0>
      outlineSettings {
       bUseBrushTransparency=false
       roundingType=HalfHeightRadius
       width=0
       color {colorUseRule=UseColor_Specified, specifiedColor=<rgba 0 0 0 0>}
       cornerRadii=<0 0 0 0>
      }
      tintColor {colorUseRule=UseColor_Specified, specifiedColor=<rgba 1 1 1 1>}
      uVRegion {bIsValid=false, max=<0 0>, min=<0 0>}
     }
     selectedTextColor {colorUseRule=UseColor_Specified, specifiedColor=<rgba 1 1 1 1>}
     selectorFocusedBrush {
      bIsDynamicallyLoaded=false
      drawAs=Border
      imageType=FullColor
      mirroring=NoMirror
      resourceName=../../../Engine/Content/Slate/Common/Selector.png
      resourceObject=null
      tiling=Both
      imageSize=<0 0>
      margin=<ltrb 0.25 0.25 0.25 0.25>
      outlineSettings {
       bUseBrushTransparency=false
       roundingType=HalfHeightRadius
       width=0
       color {colorUseRule=UseColor_Specified, specifiedColor=<rgba 0 0 0 0>}
       cornerRadii=<0 0 0 0>
      }
      tintColor {colorUseRule=UseColor_Specified, specifiedColor=<rgba 0.701 0.225 0.003 1>}
      uVRegion {bIsValid=false, max=<0 0>, min=<0 0>}
     }
     textColor {colorUseRule=UseColor_Specified, specifiedColor=<rgba 0.5271 0.5271 0.5271 1>}
    }
    widgetStyle {
     comboButtonStyle {
      downArrowAlign=VAlign_Center
      buttonStyle {
       clickedSlateSound {resourceObject=null}
       disabled {
        bIsDynamicallyLoaded=false
        drawAs=RoundedBox
        imageType=NoImage
        mirroring=NoMirror
        resourceName=None
        resourceObject=null
        tiling=NoTile
        imageSize=<0 0>
        margin=<ltrb 0 0 0 0>
        outlineSettings {
         bUseBrushTransparency=false
         roundingType=FixedRadius
         width=1
         color {colorUseRule=UseColor_ColorTable, specifiedColor=<rgba 0 0 0 0>}
         cornerRadii=<4 4 4 4>
        }
        tintColor {colorUseRule=UseColor_Specified, specifiedColor=<rgba 0.0395 0.0395 0.0395 1>}
        uVRegion {bIsValid=false, max=<0 0>, min=<0 0>}
       }
       disabledForeground {colorUseRule=UseColor_ColorTable, specifiedColor=<rgba 0 0 0 0>}
       hovered {
        bIsDynamicallyLoaded=false
        drawAs=RoundedBox
        imageType=NoImage
        mirroring=NoMirror
        resourceName=None
        resourceObject=null
        tiling=NoTile
        imageSize=<0 0>
        margin=<ltrb 0 0 0 0>
        outlineSettings {
         bUseBrushTransparency=false
         roundingType=FixedRadius
         width=1
         color {colorUseRule=UseColor_ColorTable, specifiedColor=<rgba 0 0 0 0>}
         cornerRadii=<4 4 4 4>
        }
        tintColor {colorUseRule=UseColor_Specified, specifiedColor=<rgba 0.0048 0.0048 0.0048 1>}
        uVRegion {bIsValid=false, max=<0 0>, min=<0 0>}
       }
       hoveredForeground {colorUseRule=UseColor_ColorTable, specifiedColor=<rgba 0 0 0 0>}
       hoveredSlateSound {resourceObject=null}
       normal {
        bIsDynamicallyLoaded=false
        drawAs=RoundedBox
        imageType=NoImage
        mirroring=NoMirror
        resourceName=None
        resourceObject=null
        tiling=NoTile
        imageSize=<0 0>
        margin=<ltrb 0 0 0 0>
        outlineSettings {
         bUseBrushTransparency=false
         roundingType=FixedRadius
         width=1
         color {colorUseRule=UseColor_ColorTable, specifiedColor=<rgba 0 0 0 0>}
         cornerRadii=<4 4 4 4>
        }
        tintColor {colorUseRule=UseColor_Specified, specifiedColor=<rgba 0.0048 0.0048 0.0048 1>}
        uVRegion {bIsValid=false, max=<0 0>, min=<0 0>}
       }
       normalForeground {colorUseRule=UseColor_ColorTable, specifiedColor=<rgba 0 0 0 0>}
       normalPadding=<ltrb 8 1 8 1>
       pressed {
        bIsDynamicallyLoaded=false
        drawAs=RoundedBox
        imageType=NoImage
        mirroring=NoMirror
        resourceName=None
        resourceObject=null
        tiling=NoTile
        imageSize=<0 0>
        margin=<ltrb 0 0 0 0>
        outlineSettings {
         bUseBrushTransparency=false
         roundingType=FixedRadius
         width=1
         color {colorUseRule=UseColor_ColorTable, specifiedColor=<rgba 0 0 0 0>}
         cornerRadii=<4 4 4 4>
        }
        tintColor {colorUseRule=UseColor_Specified, specifiedColor=<rgba 0.0048 0.0048 0.0048 1>}
        uVRegion {bIsValid=false, max=<0 0>, min=<0 0>}
       }
       pressedForeground {colorUseRule=UseColor_ColorTable, specifiedColor=<rgba 0 0 0 0>}
       pressedPadding=<ltrb 8 2 8 0>
       pressedSlateSound {resourceObject=null}
      }
      contentPadding=<ltrb 0 0 0 0>
      downArrowImage {
       bIsDynamicallyLoaded=false
       drawAs=Image
       imageType=Vector
       mirroring=NoMirror
       resourceName=../../../Engine/Content/Slate/Starship/CoreWidgets/ComboBox/wide-chevron-down.svg
       resourceObject=null
       tiling=NoTile
       imageSize=<16 16>
       margin=<ltrb 0 0 0 0>
       outlineSettings {
        bUseBrushTransparency=false
        roundingType=HalfHeightRadius
        width=0
        color {colorUseRule=UseColor_Specified, specifiedColor=<rgba 0 0 0 0>}
        cornerRadii=<0 0 0 0>
       }
       tintColor {colorUseRule=UseColor_Specified, specifiedColor=<rgba 1 1 1 1>}
       uVRegion {bIsValid=false, max=<0 0>, min=<0 0>}
      }
      downArrowPadding=<ltrb 2 2 2 2>
      menuBorderBrush {
       bIsDynamicallyLoaded=false
       drawAs=RoundedBox
       imageType=NoImage
       mirroring=NoMirror
       resourceName=None
       resourceObject=null
       tiling=NoTile
       imageSize=<0 0>
       margin=<ltrb 0 0 0 0>
       outlineSettings {
        bUseBrushTransparency=false
        roundingType=FixedRadius
        width=1
        color {colorUseRule=UseColor_ColorTable, specifiedColor=<rgba 0 0 0 0>}
        cornerRadii=<0 0 0 0>
       }
       tintColor {colorUseRule=UseColor_Specified, specifiedColor=<rgba 0.0395 0.0395 0.0395 1>}
       uVRegion {bIsValid=false, max=<0 0>, min=<0 0>}
      }
      menuBorderPadding=<ltrb 0 0 0 0>
      shadowColorAndOpacity=<rgba 0 0 0 1>
      shadowOffset=<0 0>
     }
     contentPadding=<ltrb 0 0 0 0>
     menuRowPadding=<ltrb 8 3 8 3>
     pressedSlateSound {resourceObject=null}
     selectionChangeSlateSound {resourceObject=null}
    }
   }
   slot {class=HorizontalBoxSlot, properties {verticalAlignment=VAlign_Center, padding=<ltrb 0 0 10 0>, size {sizeRule=Fill, value=1}}}
  }
 ]
}
