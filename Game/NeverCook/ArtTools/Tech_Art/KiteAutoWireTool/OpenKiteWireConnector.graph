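# Editor utility Blueprint that opens the Kite Wire Connector tool, part of the tech art kite auto-wire tooling.
# Its single function spawns and registers EWBP_KiteWireConnector as an editor tab through the EditorUtilitySubsystem.
# It then casts the returned widget and calls SetGrabbedActors on it; the event graph is empty.
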EditorUtilityBlueprint /Game/NeverCook/ArtTools/Tech_Art/KiteAutoWireTool/OpenKiteWireConnector
graphs [
 EventGraph {links [], nodes []}
 OpenKiteWireConnector {nodes [
   OpenKiteWireConnector() {is=title K2Node_FunctionEntry}
   > n1 = SpawnAndRegisterTab(self=Get EditorUtilitySubsystem(), InBlueprint=&/EWBP_KiteWireConnector)
   > n2 = Cast To EWBP_KiteWireConnector!(execute, Object=@n1) -> (then, CastFailed, AsEWBP Kite Wire Connector:EWBP Kite Wire Connector Object Reference) {is=title K2Node_DynamicCast}
   > SetGrabbedActors(self:EWBP Kite Wire Connector Object Reference=@n2)
  ]}
]
variables []
